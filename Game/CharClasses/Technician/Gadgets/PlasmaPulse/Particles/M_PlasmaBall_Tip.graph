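# Translucent, unlit, two-sided mesh-particle material for the plasma ball tip of the Technician's Plasma Pulse gadget.
# A panning ripple texture distorts the UVs of a scrolling plasma core texture, which is tinted by particle color and passed through emissive adjustments.
# Particle alpha, an optional sphere mask on the second UV channel and a Fresnel term form the opacity.
# A parameterized Jiggle_WPO function supplies vertex wobble.
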Material /Game/CharClasses/Technician/Gadgets/PlasmaPulse/Particles/M_PlasmaBall_Tip
graph {
 properties {
  bUsedWithMeshParticles=true
  bUsedWithNiagaraMeshParticles=true
  blendMode=BLEND_Translucent
  pixelDepthOffsetMode=PDOM_Legacy
  shadingModel=MSM_Unlit
  twoSided=true
  usedShadingModels=Unlit
 }
 nodes [
  n1 = ParticleColor()
  n2 = TextureCoordinate() {vTiling=0.5}
  n3 = Multiply(
   A=TextureSample(UVs=Panner(Coordinate=@n2) {speedX=-1, speedY=0.5}) {texture=/Game/Effects/Fort_Effects/Textures/Tile/T_RippleTile_D}.R
   B=Constant() {r=0.15}
  )
  n4 = Panner(Coordinate=Add(A=TextureCoordinate() {uTiling=2, vTiling=2}, B=Multiply(A=@n3))) {speedY=1}
  MaterialFunctionCall(Emissive=Multiply(A=@n1.RGB, B=TextureSample(UVs=@n4) {texture=/Game/Effects/Fort_Effects/Textures/Tile/T_Plasma_BallCore}.RGB)) {materialFunction=&MF_FX_EmissiveAdjustments}
  Multiply(
   A=MaterialFunctionCall!(
    Speed=ScalarParameter() {defaultValue=1, group=WPO, parameterName=Jiggle Speed, sortPriority=3}
    SmallWaveScle=ScalarParameter() {defaultValue=4, group=WPO, parameterName=Jiggle WaveScaleSmall, sortPriority=1}
    WaveScale=ScalarParameter() {defaultValue=24, group=WPO, parameterName=Jiggle WaveScale, sortPriority=2}
   ) {materialFunction=/Game/Effects/Fort_Effects/Materials/_Functions/Jiggle_WPO}.Result
   B=ScalarParameter() {defaultValue=1, group=WPO, parameterName=Jiggle Amount, sortPriority=0}
  )
  Multiply(A=@n2, B=DynamicParameter() {defaultValue=<rgba 0 0 0 1>, paramNames [UVDistortScale, UvTxtureScale, MaxDepth, Param4]}.UvTxtureScale)
  n5 = Constant3Vector() {constant=<rgba 1 1 1 0>}
  n6 = Multiply(
   A=Constant() {r=1}
   B=TextureSample(UVs=TextureCoordinate() {coordinateIndex=1}) {texture=/Game/Effects/Fort_Effects/Textures/Mask/T_Sphere_A}.RGB
  )
  n7 = ScalarParameter() {defaultValue=1, group=Opacity, parameterName=Use Circular Mask, sortPriority=1}
  n8 = ScalarParameter() {group=Opacity, parameterName=Fresnel Exponent, sortPriority=1}
  Multiply(A=@n1.A, B=Multiply(A=LinearInterpolate(A=@n5, B=@n6, Alpha=@n7) {constAlpha=0}, B=Fresnel(ExponentIn=@n8)))
 ]
}
