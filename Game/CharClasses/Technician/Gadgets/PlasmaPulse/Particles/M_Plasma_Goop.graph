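# Translucent subsurface sprite material for plasma goop particles, filed under the Technician's Plasma Pulse gadget.
# A panning noise distorts 2x2 SubUV plasma textures, masked by absolute sine of UVs to pin cell borders.
# Dynamic parameters drive distortion, tiling, time and fade depth.
# Particle color tints a cubemap-diffuse blend, while opacity math combines texture alpha, depth-texture camera fade and noise-varied depth fade.
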Material /Game/CharClasses/Technician/Gadgets/PlasmaPulse/Particles/M_Plasma_Goop
graph {
 properties {
  bUsedWithNiagaraSprites=true
  bUsedWithParticleSprites=true
  blendMode=BLEND_Translucent
  pixelDepthOffsetMode=PDOM_Legacy
  refractionMethod=RM_IndexOfRefraction
  shadingModel=MSM_Subsurface
  translucencyLightingMode=TLM_Surface
  usedShadingModels=Subsurface
 }
 nodes [
  n1 = ParticleMacroUV()
  n2 = DynamicParameter() {defaultValue=<rgba 0 0 0 1>, paramNames [UVDistortScale, UvTxtureScale, MaxDepth, Time]}
  n3 = Panner(Coordinate=TextureCoordinate() {uTiling=2, vTiling=2}, Time=@n2.Time) {speedY=1}
  Add(
   A=TextureSampleParameter2D() {parameterName=Normal, samplerType=SAMPLERTYPE_Normal, texture=/Game/Effects/Fort_Effects/Textures/Plasma/T_Plasma_SubUV_2X2_N}.RGB
   B=TextureSample(UVs=Add(A=@n1, B=@n3)) {samplerType=SAMPLERTYPE_Normal, texture=/Game/Packages/VistaAssets/Stars/NewCloudImage_n}.RGB
  )
  Constant() {r=0.1}
  n4 = ParticleColor()
  n5 = TextureCoordinate() {uTiling=0.5, vTiling=0.5}
  n6 = TextureSample(UVs=Panner(Coordinate=Multiply(A=@n5, B=@n2.UvTxtureScale), Time=@n2.Time) {speedY=0.5}) {samplerType=SAMPLERTYPE_Normal, texture=&T_Weap_Distort_N_02}
  n7 = DynamicParameter() {defaultValue=<rgba 0 0 0 1>, paramNames [UVDistortScale, UvTxtureScale, MaxDepth, Time]}
  n8 = Abs(None=Sine(None=TextureCoordinate()))
  n9 = ComponentMask(None=@n8) {r=true}
  n10 = ComponentMask(None=@n8) {g=true}
  n11 = Add(A=TextureCoordinate(), B=Multiply(A=Multiply(A=@n6.R, B=@n7.UVDistortScale), B=Multiply(A=@n9, B=@n10)))
  n12 = TextureSampleParameter2D(UVs=@n11) {parameterName=DiffuseColor, texture=/Game/Effects/Fort_Effects/Textures/Plasma/T_Plasma_SubUV_2X2}
  n13 = Multiply(
   A=TextureSampleParameter2D(UVs=@n11) {parameterName=DepthCameraFade, texture=/Game/Effects/Fort_Effects/Textures/Plasma/T_Plasma_SubUV_2X2_Depth}.RGB
   B=PixelDepth().R
  )
  n14 = Constant()
  n15 = ScalarParameter() {defaultValue=128, parameterName=CameraFadeDistance}
  n16 = Multiply(A=@n12.A, B=OneMinus(None=SphereMask(A=@n13, B=@n14, Radius=@n15) {attenuationRadius=128, hardnessPercent=10}))
  n17 = ComponentMask(None=@n6.RGB) {r=true}
  LinearInterpolate(Alpha=DepthFade(Opacity=Multiply(A=@n4.A, B=@n16), FadeDistance=LinearInterpolate(B=@n7.MaxDepth, Alpha=@n17) {constA=25, constB=150})) {constA=1, constB=0.25}
  n18 = CameraVectorWS()
  n19 = Multiply(
   A=LinearInterpolate(
    A=TextureSampleParameterCube!(UVs=@n18, Apply View MipBias) {parameterName=Cubemap, texture=/Game/CharClasses/Technician/Gadgets/BASE/Effects/T_CrystalColor_Tile}.RGB
    B=@n12.RGB
   )
   B=@n4.RGB
  )
  MaterialFunctionCall(Emissive=@n19) {materialFunction=&MF_FX_EmissiveAdjustments}
  MaterialFunctionCall(DiffuseColor=@n19, SpecularColor=Constant() {r=2}) {materialFunction=&ConvertFromDiffSpec}
  Multiply(A=@n12.RGB, B=@n4.RGB)
  TextureSample() {texture=/Game/Effects/Fort_Effects/Textures/Plasma/T_PlasmaUV_Mask}
 ]
}
