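# Translucent unlit refraction material for particle sprites and meshes, named as the shock wave of the Technician's Plasma Pulse gadget.
# It uses index-of-refraction distortion, with a sphere mask times particle alpha driving a lerp from one to zero.
# The same mask scales a sphere normal map.
# A static switch selects zero or a depth fade, which mobile feature levels replace with the raw FadeDistance parameter.
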Material /Game/CharClasses/Technician/Gadgets/PlasmaPulse/Particles/M_ShockWave_Refract
graph {
 properties {
  bUsedWithMeshParticles=true
  bUsedWithNiagaraMeshParticles=true
  bUsedWithNiagaraSprites=true
  bUsedWithParticleSprites=true
  blendMode=BLEND_Translucent
  pixelDepthOffsetMode=PDOM_Legacy
  refractionMethod=RM_IndexOfRefraction
  shadingModel=MSM_Unlit
  usedShadingModels=Unlit
 }
 nodes [
  n1 = TextureSample() {texture=/Game/Effects/Fort_Effects/Textures/Mask/T_Sphere_A}
  LinearInterpolate(Alpha=Multiply(A=ParticleColor().A, B=@n1.R)) {constA=1, constB=0}
  Multiply(A=TextureSample() {samplerType=SAMPLERTYPE_Normal, texture=/Game/Effects/Fort_Effects/Textures/Mask/T_Sphere_N}.RGB, B=@n1.R)
  n2 = ScalarParameter() {parameterName=FadeDistance}
  n3 = DepthFade(FadeDistance=@n2)
  StaticSwitchParameter(True=FeatureLevelSwitch(Default=@n3, ES2=@n2, ES3_1=@n2, SM4_REMOVED=@n3), False=Constant()) {parameterName=UseDepthFade?}
 ]
}
